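# Widgetless Editor Utility Widget that serves as a tech-art function library of Blueprint-wrapped Python scripts for Fortnite cosmetic tooling.
# Execute Python Script nodes export and import CSV files as comma-joined string arrays and read variant channel names from Athena cosmetic item definitions.
# A scratch sample function spawns a soldier mannequin for each variant combination of a character item, spaced along Y, alongside disconnected debug experiments.
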EditorUtilityWidgetBlueprint /Game/NeverCook/ArtTools/Tech_Art/Libraries/TechArt_EUW_PythonLibraryWidget
graphs [
 EventGraph {nodes [
   Event PreConstruct()
   Event Construct()
   Event Tick()
  ]}
 ExportStringArrayToCSV {nodes [
   n1 = ExportStringArrayToCSV!() -> (then, Filepath:String, CSVRows:Array of Strings) {is=title K2Node_FunctionEntry}
   > n2 = Execute Python Script!(
    execute
    PythonScript="import csv\r\ncsvfile = open(filepath, 'w', newline = '')\r\n\r\nwriter = csv.writer(csvfile)\r\n\r\nfor line in rows:\r\n    values = line.split(',')\r\n    print(values)\r\n    writer.writerow(values)\r\n"
    Rows:Array of Strings=@n1.CSVRows
    filepath:String=@n1
   ) -> (then, ReturnValue) {inputs [Rows,filepath]}
   > Return Node!(execute, Success?:bool=@n2)
  ]}
 GetItemVariantNames {nodes [
   n1 = GetItemVariantNames!() -> (then, ID:Athena Cosmetic Item Definition Object Reference) {is=title K2Node_FunctionEntry}
   > n2 = Is Valid(InputObject=@n1)
   PrintString(execute=@n2.Is Not Valid, InString="Error, Input on GetItemVariantNames Invalid")
   n3 = Execute Python Script!(
    execute=@n2.Is Valid
    PythonScript="import unreal\r\n\r\nvariant_names = []\r\n\r\nitem_variants = id.get_editor_property(\"ItemVariants\")\r\n\r\nfor variant in item_variants:\r\n    name = variant.get_editor_property(\"VariantChannelName\")\r\n    variant_names.append(str(name))"
    ID:Athena Cosmetic Item Definition Object Reference=@n1
   ) -> (then, ReturnValue, VariantNames:Array of Names) {inputs [ID], outputs [VariantNames]}
   > Return Node!(execute, Success?=@n3, ItemVariantIndexNames:Array of Names=@n3.VariantNames)
  ]}
 ImportCSV_ToStringArray {nodes [
   n1 = ImportCSV_ToStringArray!() -> (then, File Path:File Path Structure) {is=title K2Node_FunctionEntry}
   > n2 = Execute Python Script!(
    execute
    PythonScript="import csv\r\n\r\nlines = []\r\n\r\nwith open(filepath, newline ='') as csvfile:\r\n    spamreader = csv.reader(csvfile, quotechar='|')\r\n    for row in spamreader:\r\n        print(row)\r\n        ##section added to remove [' and '] from string\r\n        tempstr = str(row)\r\n        tempstr = tempstr[2:len(tempstr)-2]\r\n        print(tempstr)\r\n        lines.append(tempstr)"
    Filepath:String=Break File Path!(FilePath:"File Path Structure (by ref)"=@n1) -> (FilePath:String) {is=title K2Node_BreakStruct}
   ) -> (then, ReturnValue, Lines:Array of Strings) {inputs [Filepath], outputs [Lines]}
   > Return Node!(execute, Success?=@n2, CSV_LineArray:Array of Strings=@n2.Lines)
  ]}
 SampleWorkingSpawnAllVariants {nodes [
   n1 = SampleWorkingSpawnAllVariants!() -> (then, ID:Athena Cosmetic Item Definition Object Reference, VariantName:Name) {is=title K2Node_FunctionEntry}
   n2 = GetVariantNames!(
    execute
    ItemDefinition:Athena Cosmetic Item Definition Object Reference=@n1
    VariantChannel:Gameplay Tag Structure
    VariantTokens:Array of Fort Variant Token Type Object References
    UnlockableOnly:bool=true
   ) -> (then, ReturnValue:Array of Gameplay Tag Structures)
   > n3 = For Each Loop(Array:Array of Gameplay Tag Structures=@n2) -> (Array Element:Gameplay Tag Structure)
   > PrintString(InString="To String (Name)"(InName=GetTagName!(GameplayTag:"Gameplay Tag Structure (by ref)"=@n3) -> (ReturnValue:Name=None)))
   n4 = Is Valid(InputObject=@n1)
   PrintString(execute=@n4.Is Not Valid, InString=Error in GetVariantOptionNames)
   Execute Python Script!(
    execute=@n4.Is Valid
    PythonScript="print('Hello World!')"
    id:Athena Cosmetic Item Definition Object Reference=@n1
    variant_name:Name=@n1.VariantName
   ) -> (then, ReturnValue, option_names:Wildcard) {inputs [id,variant_name], outputs [option_names]}
   > Return Node!(execute, Success?=false, VariantOptionNames:Array of Names)
   n5 = GetAllVariantCombinations!(
    execute=@n1.then
    ItemDefinition:Athena Cosmetic Item Definition Object Reference=@n1
    IgnoredChannels:Array of Strings=Make Array!() -> (Array:Array of Strings)
    UnlockableOnly:bool=true
   ) -> (then, ReturnValue:Array of Gameplay Tag Container Structures)
   > n6 = For Each Loop(Array:Array of Gameplay Tag Container Structures=@n5) -> (Array Element:Gameplay Tag Container Structure)
   > n7 = MakeVariantChannelInfosFromVariantTagContainer!(
    execute
    ItemDefinition:Athena Cosmetic Item Definition Object Reference=@n1
    VariantTags:Gameplay Tag Container Structure=@n6
    CachedVariantTokens:Array of Fort Variant Token Type Object References
   ) -> (then, ReturnValue:Array of Mcp Variant Channel Info Structures)
   > n8 = Cast To AthenaCharacterItemDefinition!(
    execute
    Object=Break Mcp Variant Channel Info!(McpVariantChannelInfo:"Mcp Variant Channel Info Structure (by ref)"="Get (a copy)"(Array:Array of Mcp Variant Channel Info Structures=@n7) -> (Output:Mcp Variant Channel Info Structure)) -> (
     OwnedVariantTags_GameplayTags:Array of Gameplay Tag Structures
     ItemVariantIsUsedFor:Item Definition Base Object Reference
     CustomData:String
     VariantChannelTag:Gameplay Tag Structure
     ActiveVariantTag:Gameplay Tag Structure
    ) {is=title K2Node_BreakStruct}.ItemVariantIsUsedFor
   ) -> (then, CastFailed, AsAthena Character Item Definition:Athena Character Item Definition Object Reference) {is=title K2Node_DynamicCast}
   SpawnAthenaCharacterSkin!(
    execute=@n9.then
    ItemDefinition:Athena Character Item Definition Object Reference=@n8
    SpawnTransform_Location_X:double=0
    SpawnTransform_Location_Y:double=float * float(A:int=@n6.Array Index, B=100) {is=title K2Node_PromotableOperator}
    SpawnTransform_Location_Z:double=0
    SpawnTransform_Rotation:Rotator="0, 0, 0"
    SpawnTransform_Scale:Vector="1,1,1"
    VariantChannelInfos:Array of Mcp Variant Channel Info Structures=@n7
    ClearAnimations:bool=true
    MannequinClass:Object Class Reference=/Game/Characters/Player/Mannequins/SoldierMannequin_Male.SoldierMannequin_Male_C
   ) -> (then, ReturnValue:Fort Player Mannequin Object Reference)
   PrintString(execute=@n8.CastFailed, InString=Failure)
   PrintString(InString="To String (Name)"(InName=GetTagName() -> (ReturnValue=None)))
   BreakGameplayTagContainer!(GameplayTagContainer:"Gameplay Tag Container Structure (by ref)"="Get (a copy)"(Array:Array of Gameplay Tag Container Structures=@n5) -> (Output:Gameplay Tag Container Structure)) -> (GameplayTags:Array of Gameplay Tag Structures)
   GetVariantNames(ItemDefinition=@n1, UnlockableOnly=true)
   GetVariantChannels!(
    execute
    ItemDefinition:Athena Cosmetic Item Definition Object Reference=@n1
    VariantTokens:Array of Fort Variant Token Type Object References
    UnlockableOnly:bool=true
   ) -> (then, ReturnValue:Array of Gameplay Tag Structures)
   Execute Python Script!(
    execute
    PythonScript="import unreal\r\n\r\nvariant_names = []\r\n\r\nitem_variants = id.get_editor_property(\"ItemVariants\")\r\n\r\nvariant_names = item_variants[0].get_editor_property(\"VariantChannelTag\")"
    ID:Wildcard
   ) -> (then, ReturnValue, VariantNames:Wildcard) {inputs [ID], outputs [VariantNames]}
   Execute Python Script!(execute, PythonScript="import unreal\r\nprint(str(in_))", in:Gameplay Tag Structure=@n3) -> (then, ReturnValue) {inputs [in]}
   Execute Python Script!(
    execute
    PythonScript="import unreal\r\n\r\nvariant_names = []\r\n\r\nitem_variants = id.get_editor_property(\"ItemVariants\")\r\n\r\nvariant_names = item_variants[0].static_struct()\r\nprint(str(variant_names))"
    ID:Athena Cosmetic Item Definition Object Reference=@n1
   ) -> (then, ReturnValue) {inputs [ID]}
   Execute Python Script(PythonScript="import unreal\r\n\r\nprint(id)", ID:Wildcard) {inputs [ID]}
   Execute Python Script!(
    execute
    PythonScript="import unreal\r\n\r\nprint(\"Before\")\r\n\r\nunreal.SystemLibrary.execute_console_command(self,\"Editor.AsyncAssetCompilation\")\r\n\r\nprint(\"After\")"
    self:Self Object Reference=Self-Reference()
   ) -> (then, ReturnValue) {inputs [self]}
   n9 = PrintString(execute=@n8.then, InString=success)
  ]}
]
variables []
widgets {info {rootWidgetClass=null, widgetCount=0}, widgets []}
