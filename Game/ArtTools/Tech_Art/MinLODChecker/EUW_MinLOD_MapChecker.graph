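# Editor utility widget for tech art that audits static meshes on BuildingSMActors in the loaded map.
# Optional checks test Medium-quality MinLOD on higher-triangle meshes, LOD reduction percentages, and LOD-bias-adjusted sizes of material textures found through Asset Registry dependencies.
# Each component reports its first failure, logged once per asset.
# Buttons export results to CSV through a Python library widget or rebuild a local collection of flagged assets.
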EditorUtilityWidgetBlueprint /Game/ArtTools/Tech_Art/MinLODChecker/EUW_MinLOD_MapChecker
graphs [
 CheckReductionSettingsForErrors {nodes [
   n1 = CheckReductionSettingsForErrors!() -> (then, StaticMesh:Static Mesh Object Reference) {is=title K2Node_FunctionEntry}
   > n2 = Is Valid(InputObject=@n1)
   Return Node!(execute=@n2.Is Not Valid, ErrorPresent?:bool=false)
   > n3 = Sequence()
   n4 = Get StaticMeshEditorSubsystem()
   > n5 = GetLodReductionSettings!(execute, self:Static Mesh Editor Subsystem Object Reference=@n4, StaticMesh:Static Mesh Object Reference=@n1, LodIndex:int=1) -> (then, OutReductionOptions:Mesh Reduction Settings Structure)
   > Branch(Condition="float < float"(
     A=Break Mesh Reduction Settings!(MeshReductionSettings:"Mesh Reduction Settings Structure (by ref)"=@n5) -> (PercentTriangles:float=~) {is=title K2Node_BreakStruct}
     B=50
    ))
   > Return Node(ErrorPresent?=true)
   n6 = Branch(execute=@n3.then_1, Condition="integer < integer"(A=GetNumLODs(self=@n1), B=3))
   n7 = GetLodReductionSettings(execute=@n6.else, self=@n4, StaticMesh=@n1, LodIndex=2)
   > Branch(Condition="float < float"(A=Break Mesh Reduction Settings(MeshReductionSettings=@n7) -> (PercentTriangles=~), B=75))
   > Return Node(ErrorPresent?=true)
  ]}
 CheckTextureSizes {nodes [
   n1 = CheckTextureSizes!() -> (then, StaticMesh:Static Mesh Object Reference) {is=title K2Node_FunctionEntry}
   > n2 = Is Valid(InputObject=@n1)
   Return Node!(execute=@n2.Is Not Valid, HasBadTextures?:bool=false, ReportedTexture:Object Reference, CurrentTextureSize:String)
   n3 = Get StaticMeshEditorSubsystem()
   > n4 = For Loop with Break!(
    Execute:Exec
    FirstIndex:int
    LastIndex:int=int - int(A=GetNumberMaterials!(self:Static Mesh Editor Subsystem Object Reference=@n3, StaticMesh:Static Mesh Object Reference=@n1) -> (ReturnValue:int=0))
    Break:Exec=@n16.then
   ) -> (LoopBody:Exec, Index:int, Completed:Exec) {is=title K2Node_MacroInstance}
   Return Node(
    execute=@n4.Completed
    HasBadTextures?=Get BadTexturePresent?() -> (BadTexturePresent?:bool)
    ReportedTexture=Get CurrentTexture() -> (CurrentTexture:Object Reference)
    CurrentTextureSize="To String (Integer)"(InInt=Get CurrentSize() -> (CurrentSize:int=0))
   )
   n5 = GetAssetRegistry()
   > n6 = Get Dependencies!(
    execute
    self:Asset Registry Interface=@n5
    PackageName:Name=CreateAssetData!(
     InAsset=GetMaterial!(self:Static Mesh Object Reference=@n1, MaterialIndex:int=@n4) -> (ReturnValue:Material Interface Object Reference)
     bAllowBlueprintClass=false
    ) -> (
     ReturnValue_PackageName:Name
     ReturnValue_PackagePath:Name
     ReturnValue_AssetName:Name
     ReturnValue_AssetClassPath:Top Level Asset Path Structure
    )
    DependencyOptions:"Asset Registry Dependency Options Structure (by ref)"=Make Asset Registry Dependency Options(bIncludeSoftPackageReferences=True)
   ) -> (then, OutDependencies:Array of Names, ReturnValue:bool) {is=title K2Node_CallFunction}
   > n7 = For Each Loop(Array:Array of Names=@n6) -> (Array Element:Name)
   > n8 = GetAssetsByPackageName!(execute, self:Asset Registry Interface=@n5, PackageName:Name=@n7, bIncludeOnlyOnDiskAssets:bool=false, bSkipARFilteredAssets:bool=true) -> (then, OutAssetData:Array of Asset Data Structures, ReturnValue:bool)
   > n9 = For Each Loop(Array:Array of Asset Data Structures=@n8) -> (Array Element:Asset Data Structure)
   > Branch(Condition=Contains(
     SearchIn=Break Asset Data!(AssetData=@n9) -> (PackageName, PackagePath, AssetName, AssetClassPath_PathString:String).AssetClassPath_PathString
     Substring=Texture2D
    ))
   > n10 = LoadAsset(AssetPath=ToSoftObjectPath!(InAssetData:"Asset Data Structure (by ref)"=@n9) -> (ReturnValue_PathString:String))
   > Set CurrentTexture(CurrentTexture:Object Reference=@n10) -> (Output_Get:Object Reference)
   > n11 = Cast To Texture!(execute, Object=@n10) -> (then, CastFailed, AsTexture:Texture Object Reference) {is=title K2Node_DynamicCast}
   > n12 = Get Texture Size!(execute, Texture:Texture Object Reference=@n11, __WorldContext:Object Reference) -> (then, Size:Int Point Structure) {is=title K2Node_CallFunction}
   n13 = Get LODBias!(self:Texture Object Reference=@n11) -> (LODBias:int=0)
   n14 = IntPoint / Integer!(
    A:Int Point Structure=@n12
    B:int="Max (Integer)"(A=int * int(A:int=@n13, B:int=2) -> (ReturnValue:int=0) {is=title K2Node_CommutativeAssociativeBinaryOperator}, B=1)
   ) -> (ReturnValue_X:int, ReturnValue_Y:int)
   n15 = "Max (Integer)"(A=@n14, B=@n14.ReturnValue_Y)
   > Branch(Condition="integer >= integer"(A=@n15, B=4000))
   > Set CurrentSize(CurrentSize:int=@n15) -> (Output_Get:int=0)
   > n16 = Set BadTexturePresent?(BadTexturePresent?:bool=true) -> (Output_Get:bool)
   GetDisplayName(Object=@n10)
  ]}
 EventGraph {nodes [
   "On Clicked (CheckMap)"!() -> (OutputDelegate, then, widget:User Widget Object Reference, event:Pointer Event Structure) {is=title K2Node_ComponentBoundEvent}
   > Clear(TargetArray:Array of Struct Min LOD Reporting Data Structures=Get CSVReportingData() -> (CSVReportingData:Array of Struct Min LOD Reporting Data Structures))
   > Clear(TargetArray:Array of Object References=Get AssetsToInvestigate() -> (AssetsToInvestigate:Array of Object References))
   > n1 = GetAllActorsOfClass!(execute, WorldContextObject:Object Reference, ActorClass:Actor Class Reference=&Actor) -> (then, OutActors:Array of Actor Object References)
   > Set LoadedActorCount(LoadedActorCount:int=Length(TargetArray=@n1)) -> (Output_Get:int=0)
   > n2 = For Each Loop(Array:Array of Actor Object References=@n1) -> (Array Element:Actor Object Reference)
   > n3 = Cast To BuildingSMActor!(execute, Object=@n2) -> (then, CastFailed, AsBuilding Static Mesh:Building Static Mesh Object Reference) {is=title K2Node_DynamicCast}
   n4 = For Each Loop(
    Exec=[@n10.CastFailed, @n3.then]
    Array:Array of Static Mesh Component Object References=Get Components By Class!(self:Actor Object Reference=@n2, ComponentClass:Actor Component Class Reference=&StaticMeshComponent) -> (ReturnValue:Array of Static Mesh Component Object References)
   ) -> (Array Element:Static Mesh Component Object Reference)
   > n5 = GetStaticMeshQLMinLOD!(execute, self, StaticMeshComponent:Static Mesh Component Object Reference=@n4) -> (then, MinLODAtQL:int=0)
   > n6 = Reporting!(execute, self, StaticMeshComponent:Static Mesh Component Object Reference=@n4, MinLOD:int=@n5) -> (
    then
    MarkedForInvestigation?:bool
    MarkedReason:EUW_MinLOD_ErrorEnums Enum=NewEnumerator0
    ReportedAsset:Object Reference
    AdditionalInformation:String
   )
   > Logging!(
    execute
    self
    RequiresLogging?:bool=@n6
    ReportingReason:EUW_MinLOD_ErrorEnums Enum=@n6.MarkedReason
    ReportedAsset:Object Reference=@n6.ReportedAsset
    AdditionalInformation:String=@n6.AdditionalInformation
   ) -> (then)
   "On Clicked (SaveCSV)"!() -> (OutputDelegate, then, widget:User Widget Object Reference, event:Pointer Event Structure) {is=title K2Node_ComponentBoundEvent}
   > SaveToCSV()
   "On Clicked (SaveCollection)"!() -> (OutputDelegate, then, widget:User Widget Object Reference, event:Pointer Event Structure) {is=title K2Node_ComponentBoundEvent}
   > SaveToCollection()
   n7 = "On Selection Changed (CheckMinLOD)"!() -> (OutputDelegate, then, selected:bool, widget:User Widget Object Reference) {is=title K2Node_ComponentBoundEvent}
   > Set CheckQLMLod(CheckQLMLod:bool=@n7.selected) -> (Output_Get:bool)
   n8 = "On Selection Changed (CheckReduction)"!() -> (OutputDelegate, then, selected:bool, widget:User Widget Object Reference) {is=title K2Node_ComponentBoundEvent}
   > Set CheckReductions(CheckReductions:bool=@n8.selected) -> (Output_Get:bool)
   n9 = "On Selection Changed (CheckTextureSize)"!() -> (OutputDelegate, then, selected:bool, widget:User Widget Object Reference) {is=title K2Node_ComponentBoundEvent}
   > Set CheckTexture(CheckTexture:bool=@n9.selected) -> (Output_Get:bool)
   n10 = Cast To StaticMeshActor!(execute, Object=@n2) -> (then, CastFailed, AsStatic Mesh Actor:Static Mesh Actor Object Reference) {is=title K2Node_DynamicCast}
   n11 = Get StaticMeshComponent(self=@n10)
   > n12 = GetStaticMeshQLMinLOD(StaticMeshComponent=@n11) -> (MinLODAtQL=0)
   > n13 = Reporting(StaticMeshComponent=@n11, MinLOD=@n12) -> (MarkedReason=NewEnumerator0)
   > Logging(RequiresLogging?=@n13, ReportingReason=@n13.MarkedReason, ReportedAsset=@n13.ReportedAsset, AdditionalInformation=@n13.AdditionalInformation)
   n14 = Cast To BuildingLevelInstance!(execute, Object=@n2) -> (then, CastFailed, AsBuilding Level Instance:Building Level Instance Object Reference) {is=title K2Node_DynamicCast}
   > n15 = For Each Loop(Array:Array of Actor Object References=GetAllChildActors!(self:Actor Object Reference=@n14, bIncludeDescendants:bool=true) -> (ChildActors:Array of Actor Object References)) -> (Array Element:Actor Object Reference)
   > PrintString(InString=GetDisplayName(Object=@n15))
   Event Construct()
   > SetObject(self=Get DetailsView_48() -> (DetailsView_48:Details View Object Reference), NewObject=Self-Reference())
   GetLoadedLevel!(self:Level Instance Interface Interface=@n14) -> (ReturnValue:Level Object Reference)
   Event PreConstruct()
   Event Tick()
   Get SubobjectDataSubsystem()
   Reroute Node(InputPin:Actor Object Reference=@n2) -> (OutputPin:Actor Object Reference)
   GetLoadedLevel!(execute:Exec, self:Object Reference=@n2) -> (then:Exec, ReturnValue:Level Object Reference) {is=title K2Node_Message}
  ]}
 GetStaticMeshQLMinLOD {nodes [
   n1 = GetStaticMeshQLMinLOD!() -> (then, StaticMeshComponent:Static Mesh Component Object Reference) {is=title K2Node_FunctionEntry}
   > Return Node!(
    execute
    MinLODAtQL:int=GetMinimumLODForQualityLevel!(self:Static Mesh Object Reference=Get StaticMesh(self=@n1), QualityLevel:"Name (by ref)"=MakeLiteralName(Value=Medium)) -> (ReturnValue:int=0)
   )
  ]}
 Logging {nodes [
   n1 = Logging!() -> (
    then
    RequiresLogging?:bool=~
    ReportingReason:EUW_MinLOD_ErrorEnums Enum=NewEnumerator0
    ReportedAsset:Object Reference
    AdditionalInformation:String
   ) {is=title K2Node_FunctionEntry}
   > Branch(Condition=@n1)
   n2 = Get AssetsToInvestigate()
   > n3 = Add Unique(TargetArray:Array of Object References=@n2, NewItem:"Object Reference (by ref)"=@n1.ReportedAsset)
   > Branch(Condition=Is Valid Index!(TargetArray:Array of Object References=@n2, IndexToTest:int=@n3) -> (ReturnValue:bool))
   > Add!(
    execute
    TargetArray:Array of Struct Min LOD Reporting Data Structures=Get CSVReportingData()
    NewItem_Asset_10_#1:Object Reference=@n1.ReportedAsset
    NewItem_AssetPath_5_#2:String=GetSystemPath!(Object:Object Reference=@n1.ReportedAsset) -> (ReturnValue:String)
    NewItem_ReportingReason_4_#3:String=Enum to String!(Enumerator:EUW_MinLOD_ErrorEnums Enum=@n1.ReportingReason) -> (ReturnValue:String)
    NewItem_VertexCountOrTextureSize_7_#4:String=@n1.AdditionalInformation
   ) -> (then, ReturnValue) {is=title K2Node_CallArrayFunction}
  ]}
 Reporting {nodes [
   n1 = Reporting!() -> (then, StaticMeshComponent:Static Mesh Component Object Reference, MinLOD:int) {is=title K2Node_FunctionEntry}
   > n2 = Sequence!(execute) -> (then_0, then_1, then_2:Exec, then_3:Exec)
   > Branch(Condition=Get CheckQLMLod() -> (CheckQLMLod:bool))
   > Branch(Condition="integer < integer"(A=@n1.MinLOD))
   n3 = Get StaticMesh(self=@n1)
   n4 = GetNumTriangles!(self:Static Mesh Object Reference=@n3, LODIndex:int=0) -> (ReturnValue:int=0)
   > n5 = Branch(Condition="integer > integer"(A=@n4, B=500))
   > Return Node!(
    execute
    MarkedForInvestigation?:bool=true
    MarkedReason:EUW_MinLOD_ErrorEnums Enum=NewEnumerator0
    ReportedAsset:Object Reference=@n3
    AdditionalInformation:String="To String (Integer)"(InInt=@n4)
   )
   Return Node(execute=[@n5.else, @n2.then_3], MarkedForInvestigation?=false, MarkedReason=NewEnumerator0)
   Branch(execute=@n2.then_1, Condition=Get CheckReductions() -> (CheckReductions:bool))
   > n6 = CheckReductionSettingsForErrors!(execute, self, StaticMesh:Static Mesh Object Reference=@n3) -> (then, ErrorPresent?:bool) {is=title K2Node_CallFunction}
   > Branch(Condition=@n6)
   > Return Node(MarkedForInvestigation?=true, MarkedReason=NewEnumerator1, ReportedAsset=@n3)
   Branch(execute=@n2.then_2, Condition=Get CheckTexture() -> (CheckTexture:bool))
   > n7 = CheckTextureSizes!(execute, self, StaticMesh:Static Mesh Object Reference=@n3) -> (then, HasBadTextures?:bool, ReportedTexture:Object Reference, CurrentTextureSize:String) {is=title K2Node_CallFunction}
   > Branch(Condition=@n7)
   > Return Node(MarkedForInvestigation?=true, MarkedReason=NewEnumerator2, ReportedAsset=@n7.ReportedTexture, AdditionalInformation=@n7.CurrentTextureSize)
   GetDisplayName(Object=@n3)
   Get CheckMinLOD() -> (CheckMinLOD:TA Generic 01 Checkbox Object Reference)
  ]}
 SaveToCSV {nodes [
   SaveToCSV() {is=title K2Node_FunctionEntry}
   n1 = Get CSVRows() -> (CSVRows:Array of Strings)
   > Add!(execute, TargetArray:Array of Strings=@n1, NewItem:"String (by ref)"="AssetPath, ReportingReason, Vertex Count or Texture Size") -> (then, ReturnValue)
   > n2 = For Each Loop!(Exec, Array:Array of Struct Min LOD Reporting Data Structures=Get CSVReportingData()) -> (
    LoopBody
    Array Element_Asset_10_#1:Object Reference
    Array Element_AssetPath_5_#2:String
    Array Element_ReportingReason_4_#3:String
    Array Element_VertexCountOrTextureSize_7_#4:String
    Array Index
    Completed
   )
   ExportStringArrayToCSV!(
    execute=@n2.Completed
    self:Tech Art EUW Python Library Widget Object Reference=Get TechArt_EUW_PythonLibraryWidget() -> (TechArt_EUW_PythonLibraryWidget:Tech Art EUW Python Library Widget Object Reference)
    Filepath:String=Append(A=Get!() -> (OutputPath_Path:String), B=/MinLODMapCheck.CSV)
    CSVRows:Array of Strings=Get CSVRows()
   ) -> (then, Success?:bool)
   Add(
    execute=@n2.LoopBody
    TargetArray=@n1
    NewItem=Append!(
     A=@n2.Array Element_AssetPath_5_#2
     B=","
     C:String=@n2.Array Element_ReportingReason_4_#3
     D:String=","
     E:String=@n2.Array Element_VertexCountOrTextureSize_7_#4
    ) -> (ReturnValue)
   )
  ]}
 SaveToCollection {nodes [
   SaveToCollection() {is=title K2Node_FunctionEntry}
   n1 = Get CollectionManagerScriptingSubsystem()
   n2 = Get CollectionName() -> (CollectionName:Name=None)
   > n3 = CollectionExists!(
    execute
    self:Collection Manager Subsystem Object Reference=@n1
    Container:Collection Container Source Structure
    Collection:Name=@n2
    ShareType:ECollectionScriptingShareType Enum=Local
   ) -> (then, ReturnValue:bool)
   > n4 = Branch(Condition=@n3)
   > n5 = GetCollectionsByName!(execute, self:Collection Manager Subsystem Object Reference=@n1, Container:Collection Container Source Structure, Collection:Name=@n2) -> (then, OutCollections:Array of Collection Structures, ReturnValue:bool)
   > n6 = DestroyCollection!(
    execute
    self:Collection Manager Subsystem Object Reference=@n1
    Collection:"Collection Structure (by ref)"="Get (a copy)"(Array:Array of Collection Structures=@n5) -> (Output:Collection Structure)
   ) -> (then, ReturnValue:bool)
   n7 = CreateCollection!(
    execute=[@n6.then, @n4.else]
    self:Collection Manager Subsystem Object Reference=@n1
    Container:Collection Container Source Structure
    Collection:Name=@n2
    ShareType:ECollectionScriptingShareType Enum=Local
   ) -> (then, OutNewCollection:Collection Structure, ReturnValue:bool)
   > n8 = For Each Loop!(Exec, Array:Array of Object References=Get AssetsToInvestigate()) -> (LoopBody, Array Element:Object Reference, Array Index, Completed)
   > AddAssetToCollection!(
    execute
    self:Collection Manager Subsystem Object Reference=@n1
    Collection:"Collection Structure (by ref)"=@n7
    AssetPath_PathString:String=Get Object Path String(Object=@n8)
   ) -> (then, ReturnValue:bool)
   Get AssetTagsSubsystem() -> (ReturnValue:Asset Tags Subsystem Object Reference) {is=title K2Node_GetEngineSubsystem}
  ]}
]
variables [
 AssetsToInvestigate {type="array<Object>", default []}
 CheckQLMLod {type=boolean, default=true}
 CheckReductions {type=boolean}
 CheckTexture {type=boolean}
 CollectionName {type=string, default=Local_MinLOD_MapCheck}
 CSVReportingData {type="array<Struct_MinLOD_ReportingData>", default []}
 LoadedActorCount {type=integer}
 OutputPath {type=DirectoryPath, default {path=C:/DEFAULTDONTUSE}}
]
widgets {
 info {rootWidgetClass=&Overlay, widgetCount=14}
 widgets [
  Overlay_19 {widgetClassPath=&Overlay, widget=Overlay_19}
  VerticalBox_0 {
   widgetClassPath=&VerticalBox
   parent=Overlay_19
   widget=VerticalBox_0
   slot {class=OverlaySlot, properties {horizontalAlignment=HAlign_Fill, verticalAlignment=VAlign_Fill}}
  }
  HorizontalBox_0 {
   widgetClassPath=&HorizontalBox
   parent=VerticalBox_0
   widget=HorizontalBox_0
   slot {class=VerticalBoxSlot, properties {horizontalAlignment=HAlign_Left}}
  }
  CheckMap {
   widgetClassPath=&TA_Generic01_Button
   parent=HorizontalBox_0
   widget=CheckMap
   properties {label=Check Loaded Region}
   slot {class=HorizontalBoxSlot}
  }
  SaveCSV {
   widgetClassPath=&TA_Generic01_Button
   parent=HorizontalBox_0
   widget=SaveCSV
   properties {label=Save Results To CSV}
   slot {class=HorizontalBoxSlot}
  }
  SaveCollection {
   widgetClassPath=&TA_Generic01_Button
   parent=HorizontalBox_0
   widget=SaveCollection
   properties {label=Save Results To Collection}
   slot {class=HorizontalBoxSlot}
  }
  CheckMinLOD {
   widgetClassPath=/Game/Blueprints/Blutilities/ForTechArt/Widgets/Buttons/TA_Generic01_Checkbox
   bIsVariable=true
   parent=HorizontalBox_0
   widget=CheckMinLOD
   properties {isSelected=true, label=CheckMedium QL MinLOD present?}
   slot {class=HorizontalBoxSlot}
  }
  Spacer_79 {
   widgetClassPath=&Spacer
   parent=HorizontalBox_0
   widget=Spacer_79
   slot {class=HorizontalBoxSlot, properties {padding=<ltrb 10 10 10 10>}}
  }
  CheckReduction {
   widgetClassPath=&TA_Generic01_Checkbox
   bIsVariable=true
   parent=HorizontalBox_0
   widget=CheckReduction
   properties {label=Check Correct Reduction Settings?}
   slot {class=HorizontalBoxSlot}
  }
  Spacer_131 {
   widgetClassPath=&Spacer
   parent=HorizontalBox_0
   widget=Spacer_131
   slot {class=HorizontalBoxSlot, properties {padding=<ltrb 10 10 10 10>}}
  }
  CheckTextureSize {
   widgetClassPath=&TA_Generic01_Checkbox
   bIsVariable=true
   parent=HorizontalBox_0
   widget=CheckTextureSize
   properties {label=Check Texture Size above 4000?}
   slot {class=HorizontalBoxSlot}
  }
  EditorUtilityScrollBox_0 {
   widgetClassPath=&EditorUtilityScrollBox
   parent=VerticalBox_0
   widget=EditorUtilityScrollBox_0
   slot {class=VerticalBoxSlot, properties {size {sizeRule=Fill, value=1}}}
  }
  DetailsView_48 {
   widgetClassPath=&DetailsView
   parent=EditorUtilityScrollBox_0
   widget=DetailsView_48
   properties {categoriesToShow [NoCategory]}
   slot {class=ScrollBoxSlot}
  }
  TechArt_EUW_PythonLibraryWidget {
   widgetClassPath=/Game/NeverCook/ArtTools/Tech_Art/Libraries/TechArt_EUW_PythonLibraryWidget
   bIsVariable=true
   parent=Overlay_19
   widget=TechArt_EUW_PythonLibraryWidget
   slot {class=OverlaySlot}
  }
 ]
}
